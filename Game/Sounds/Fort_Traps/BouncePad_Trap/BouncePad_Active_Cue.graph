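# Sound cue for the bounce pad trap's activation sound.
# A quality level node routes two levels to the standard wave and two to its "opt" variant, each through its own modulator.
# It uses trap attenuation and the trap placement sound class, with boosted volume and a concurrency override that prevents new instances once the limit is reached.
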SoundCue /Game/Sounds/Fort_Traps/BouncePad_Trap/BouncePad_Active_Cue
root=QualityLevel_0
nodes [
 {id=Modulator_0, children [WavePlayer_0]}
 {id=Modulator_1, children [WavePlayer_1]}
 {id=QualityLevel_0, children [Modulator_0,Modulator_0,Modulator_1,Modulator_1]}
 {id=WavePlayer_0, properties {soundWaveAssetPtr=&/BouncePad_Activate_03}}
 {id=WavePlayer_1, properties {soundWaveAssetPtr=&/BouncePad_Activate_03_opt}}
]
properties {
 attenuationSettings=&Fort_Traps_Att
 bOverrideConcurrency=true
 maxDistance=6000
 soundClassObject=/Game/Sounds/FortSoundClassesAndModes/Fort_Traps_Placement
 volumeMultiplier=1.5
 concurrencyOverrides {resolutionRule=PreventNew}
}
