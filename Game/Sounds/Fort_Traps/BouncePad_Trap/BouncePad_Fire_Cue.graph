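# Sound cue for the fire sound of the bounce pad trap, using bounce pad jump recordings.
# A quality level node sends its first two levels to the standard waves and the last two to the "opt" variants.
# Each branch passes an equally weighted random pick of two waves through a modulator.
# Trap attenuation applies, and a concurrency override caps simultaneous instances, stopping the oldest.
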SoundCue /Game/Sounds/Fort_Traps/BouncePad_Trap/BouncePad_Fire_Cue
root=QualityLevel_0
nodes [
 {id=Modulator_0, children [Random_1]}
 {id=Modulator_12, children [Random_0]}
 {id=QualityLevel_0, children [Modulator_12,Modulator_12,Modulator_0,Modulator_0]}
 {id=Random_0, children [WavePlayer_0,WavePlayer_1], properties {weights [1,1]}}
 {id=Random_1, children [WavePlayer_2,WavePlayer_3], properties {weights [1,1]}}
 {id=WavePlayer_0, properties {soundWaveAssetPtr=&/BouncePad_Jump_05}}
 {id=WavePlayer_1, properties {soundWaveAssetPtr=&/BouncePad_Jump_06}}
 {id=WavePlayer_2, properties {soundWaveAssetPtr=&/BouncePad_Jump_05_opt}}
 {id=WavePlayer_3, properties {soundWaveAssetPtr=&/BouncePad_Jump_06_opt}}
]
properties {
 attenuationSettings=&Fort_Traps_Att
 bOverrideConcurrency=true
 maxDistance=6000
 soundClassObject=&Fort_WeaponMelee_PSM
 volumeMultiplier=1
 concurrencyOverrides {maxCount=2, resolutionRule=StopOldest, volumeScale=0.5}
}
