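# Sound cue that plays when a bounce pad trap is placed, using trap attenuation and the trap placement sound class.
# A quality level node sends higher tiers to one modulated random pair of placement waves and lower tiers to a pair of "opt" variants.
# Each branch picks evenly between two waves and passes it through a modulator.
# A concurrency override caps simultaneous plays at a small count, stopping the quietest.
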SoundCue /Game/Sounds/Fort_Traps/BouncePad_Trap/BouncePad_Placed_Cue
root=QualityLevel_0
nodes [
 {id=Modulator_0, children [Random_0]}
 {id=Modulator_1, children [Random_1]}
 {id=QualityLevel_0, children [Modulator_0,Modulator_0,Modulator_1,Modulator_1]}
 {id=Random_0, children [WavePlayer_0,WavePlayer_1], properties {weights [1,1]}}
 {id=Random_1, children [WavePlayer_2,WavePlayer_3], properties {weights [1,1]}}
 {id=WavePlayer_0, properties {soundWaveAssetPtr=&/BouncePad_Place_01}}
 {id=WavePlayer_1, properties {soundWaveAssetPtr=&/BouncePad_Place_02}}
 {id=WavePlayer_2, properties {soundWaveAssetPtr=&/BouncePad_Place_01_opt}}
 {id=WavePlayer_3, properties {soundWaveAssetPtr=&/BouncePad_Place_02_opt}}
]
properties {
 attenuationSettings=&Fort_Traps_Att
 bOverrideConcurrency=true
 maxDistance=6000
 soundClassObject=/Game/Sounds/FortSoundClassesAndModes/Fort_Traps_Placement
 volumeMultiplier=0.9
 concurrencyOverrides {maxCount=2, resolutionRule=StopQuietest}
}
