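# Sound cue for the Bounce Pad trap's UI select sound.
# A single wave player feeds the Bounce_Pad_UI_Select wave straight to the output, with no randomization or modulation.
# It routes through the non-critical in-game UI sound class at reduced volume with a very large max distance.
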SoundCue /Game/Sounds/Fort_Traps/BouncePad_Trap/Bounce_Pad_UI_Select_Cue
root=WavePlayer_0
nodes [{id=WavePlayer_0, properties {soundWaveAssetPtr=&/Bounce_Pad_UI_Select}}]
properties {maxDistance=2097152, soundClassObject=&Fort_UI_InGame_NonCritical, volumeMultiplier=0.3}
